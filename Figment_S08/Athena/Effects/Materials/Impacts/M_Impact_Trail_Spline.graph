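# Translucent unlit particle material for impact smoke trails, using SplineThicken for width and panning ribbon smoke textures.
# Opacity combines squared smoke, side/end gradient masks and optional depth fade; color adds two dissolve glows to time-of-day tints with fake sun shading, remapped to HDR.
# Particle color drives UV offset, both dissolves and alpha; dynamic parameters drive pan, mip bias, thickness and noise displacement along normals.
# Mobile feature levels skip depth fade and shading.
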Material /Figment_S08/Athena/Effects/Materials/Impacts/M_Impact_Trail_Spline
graph {
 properties {
  bUsedWithMeshParticles=true
  bUsedWithParticleSprites=true
  blendMode=BLEND_Translucent
  pixelDepthOffsetMode=PDOM_Legacy
  shadingModel=MSM_Unlit
  twoSided=true
  usedShadingModels=Unlit
 }
 nodes [
  n1 = Constant() {r=1}
  n2 = ScalarParameter() {defaultValue=1.4, group=Other Settings, parameterName=V_Scale}
  n3 = Multiply(A=Multiply(A=TextureCoordinate(), B=AppendVector(A=@n1, B=@n2)), B=Constant2Vector() {g=-1, r=1})
  n4 = AppendVector(A=Constant(), B=ParticleColor().R)
  n5 = Add(
   A=MaterialFunctionCall(XY=Add(A=@n3, B=@n4)) {materialFunction=&Swizzle}.YX
   B=AppendVector(A=DynamicParameter() {defaultValue=<rgba 0 1 1 1>, paramNames [PanSpeed, MipBias, SplineThicknessMultiplier, Param4]}.PanSpeed, B=Constant())
  )
  n6 = DynamicParameter() {defaultValue=<rgba 0 1 1 1>, paramNames [PanSpeed, MipBias, SplineThicknessMultiplier, Param4]}
  n7 = LinearInterpolate(Alpha=@n6.MipBias) {constA=3, constB=0}
  n8 = TextureSample!(UVs=Add(A=@n5, B=Constant2Vector() {g=0.025, r=-0.02}), Tex, Bias=@n7, Apply View MipBias) {mipValueMode=TMVM_MipBias, texture=/Game/Effects/Fort_Effects/Gameplay/SmokeDustExplosions/Textures/T_RibbonSmokeMasked_02}
  n9 = Constant() {r=0.5}
  n10 = AppendVector(
   A=@n9
   B=MaterialFunctionCall(Float2=TextureCoordinate()) {materialFunction=&BreakOutFloat2Components}.R
  )
  n11 = ScalarParameter() {defaultValue=0.5, group=Gradient Mask Settings, parameterName=Side Mask Radius, sortPriority=4}
  n12 = ScalarParameter() {defaultValue=0.9, group=Gradient Mask Settings, parameterName=Side Mask Hardness, sortPriority=4}
  n13 = MaterialFunctionCall(Float2=TextureCoordinate()) {materialFunction=&BreakOutFloat2Components}
  n14 = OneMinus(None=@n13.G)
  n15 = ScalarParameter() {defaultValue=2, group=Gradient Mask Settings, parameterName=Bottom Mask Softness, sortPriority=4}
  n16 = Saturate(None=Multiply(
    A=Power(Base=@n14, Exp=@n15)
    B=ScalarParameter() {defaultValue=12, group=Gradient Mask Settings, parameterName=Bottom Mask Multiplier, sortPriority=4}
   ))
  n17 = ScalarParameter() {defaultValue=0.2, group=Gradient Mask Settings, parameterName=Top Mask Softness, sortPriority=4}
  n18 = Saturate(None=Multiply(
    A=Power(Base=@n13.G, Exp=@n17)
    B=ScalarParameter() {defaultValue=2, group=Gradient Mask Settings, parameterName=Top Mask Multiplier, sortPriority=4}
   ))
  n19 = Multiply(A=Multiply(A=@n8.G, B=@n8.G), B=Multiply(A=SphereMask(A=@n10, B=@n9, Radius=@n11, Hardness=@n12), B=Multiply(A=@n16, B=@n18)))
  n20 = ScalarParameter() {defaultValue=50, group=Other Settings, parameterName=Depth Fade}
  n21 = StaticSwitchParameter(True=Multiply(A=@n19, B=DepthFade(FadeDistance=@n20)), False=@n19) {group=01-Features, parameterName=Enable Depth Fade}
  Multiply(A=ParticleColor().A, B=FeatureLevelSwitch(Default=@n21, ES2=@n19, ES3_1=@n19, SM4_REMOVED=@n21, SM5=@n21, SM6=@n21))
  n22 = VectorParameter() {group=Colors - ToD, parameterName=ToD Color Morning, defaultValue=<rgba 1 0.5451 0.4 0>}
  n23 = VectorParameter() {group=Colors - Base, parameterName=Base Color Dark, sortPriority=2, defaultValue=<rgba 0.1615 0.1615 0.1615 1>}
  n24 = TextureSample(UVs=@n5, Bias=@n7) {
   mipValueMode=TMVM_MipBias
   samplerType=SAMPLERTYPE_LinearColor
   texture=/Game/Effects/Fort_Effects/Gameplay/SmokeDustExplosions/Textures/T_RibbonSmokeMasked_03
  }
  n25 = LinearInterpolate(
   A=@n23.RGB
   B=VectorParameter() {group=Colors - Base, parameterName=Base Color Light, sortPriority=2, defaultValue=<rgba 1 1 1 1>}.RGB
   Alpha=@n24.G
  ) {constB=3}
  n26 = VectorParameter() {group=Colors - ToD, parameterName=ToD ColorDay, sortPriority=1, defaultValue=<rgba 1 1 1 1>}
  n27 = VectorParameter() {group=Colors - ToD, parameterName=ToD Color Evening, sortPriority=2, defaultValue=<rgba 1 0.4263 0.4 0>}
  n28 = VectorParameter() {group=Colors - ToD, parameterName=ToD Color Night, sortPriority=3, defaultValue=<rgba 0.4 0.5658 1 0>}
  n29 = StaticSwitchParameter(
   True=MaterialFunctionCall(
    Color_Morning=Multiply(A=@n22.RGB, B=@n25)
    Color_Day=Multiply(A=@n26.RGB, B=@n25)
    Color_Evening=Multiply(A=@n25, B=@n27.RGB)
    Color_Night=Multiply(A=@n25, B=@n28.RGB)
   ) {materialFunction=&Color_Time_of_Day}.Result
   False=@n25
  ) {defaultValue=true, group=01-Features, parameterName=Enable ToD Colors}
  n30 = Constant() {r=1}
  n31 = CollectionParameter() {
   collection=&FortniteMaterialParameters
   parameterName=SunlightDirectionalVector
  }
  n32 = DynamicParameter() {defaultValue=<rgba 0 1 1 1>, paramNames [PanSpeed, MipBias, SplineThicknessMultiplier, Param4]}
  n33 = TextureSample(UVs=@n5, Bias=LinearInterpolate(Alpha=@n6.MipBias) {constA=2, constB=0}) {
   mipValueMode=TMVM_MipBias
   samplerType=SAMPLERTYPE_Normal
   texture=/Game/Effects/Fort_Effects/Gameplay/SmokeDustExplosions/Textures/T_RibbonSmokeMasked_01
  }
  n34 = MaterialFunctionCall(
   WidthBase=Multiply(
    A=@n32.SplineThicknessMultiplier
    B=ScalarParameter() {defaultValue=24, group=Spline Settings, parameterName=Spline Width Base, sortPriority=1}
   )
   WidthTip=Multiply(
    A=@n32.SplineThicknessMultiplier
    B=ScalarParameter() {defaultValue=6, group=Spline Settings, parameterName=Spline Width Tip, sortPriority=1}
   )
   AngleCorrectedNormals=StaticBool() {value=true}
   AdditionalNormal=AppendVector(A=AppendVector(A=@n33.G, B=@n33.G), B=@n33.B)
  ) {materialFunction=&SplineThicken}
  n35 = DotProduct(A=ComponentMask(None=@n31) {b=true, g=true, r=true}, B=@n34.Normal)
  n36 = Clamp(
   None=Max(A=Saturate(None=@n35)) {constB=0}
   Min=ScalarParameter() {defaultValue=0.05, group=Colors - ToD, parameterName=Shadow Min Clamp, sortPriority=3}
  )
  n37 = FeatureLevelSwitch(Default=@n36, ES2=@n30, ES3_1=@n30, SM4_REMOVED=@n36, SM5=@n36, SM6=@n36)
  n38 = VectorParameter() {group=Colors - Dissolve, parameterName=Secondary Dissolve Color, sortPriority=1, defaultValue=<rgba 35 4 0.2 1>}
  n39 = TextureSample(UVs=@n5) {
   mipValueMode=TMVM_MipBias
   samplerType=SAMPLERTYPE_LinearColor
   texture=&T_RibbonSmokeMasked_03
  }
  n40 = ComponentMask(None=TextureCoordinate()) {g=true}
  n41 = MaterialFunctionCall!(Ramp=Multiply(A=OneMinus(None=@n39.G), B=OneMinus(None=@n40)), Dissolve=ParticleColor().B, EdgeSharpness) {materialFunction=/Game/Effects/Fort_Effects/Materials/_Functions/MF_Dissolve}
  n42 = MaterialFunctionCall(Ramp=Multiply(A=@n40, B=OneMinus(None=@n24.G)), Dissolve=ParticleColor().G) {materialFunction=&MF_Dissolve}
  n43 = Add(
   A=Multiply(A=@n38.RGB, B=Ceil(None=@n41.ClampedResult))
   B=LinearInterpolate(
    A=VectorParameter() {group=Colors - Dissolve, parameterName=Primary Dissolve Color, defaultValue=<rgba 35 4 0.2 1>}.RGB
    Alpha=Ceil(None=@n42.ClampedResult)
   ) {constB=0}
  )
  n44 = StaticSwitchParameter(True=Add(A=@n43, B=Multiply(A=@n29, B=@n37)), False=Add(A=@n43, B=@n29)) {group=01-Features, parameterName="Enable Fake Directional Light "}
  Reroute(None=MaterialFunctionCall(
    LDR Input=Clamp(None=@n44)
    HDR Tint=@n44
    HDR Intensity Multiplier=ScalarParameter() {defaultValue=1, parameterName=HDR Intensity Multiplier}
   ) {materialFunction=&DeriveHDRfromLDR}.RemappedRange)
  n45 = Multiply(
   A=TextureCoordinate() {uTiling=0.5, vTiling=0.5}
   B=ScalarParameter() {defaultValue=1, group=Other Settings, parameterName=Distortion UV Scale}
  )
  n46 = Multiply(
   A=VertexNormalWS()
   B=TextureSample!(UVs=Panner(Coordinate=@n45) {bFractionalPart=true, speedX=0.125}, Tex, Apply View MipBias) {samplerType=SAMPLERTYPE_LinearColor, texture=/Game/Effects/Fort_Effects/Creatures/Shielder/T_ShielderShieldNorm}.G
  )
  n47 = Multiply(
   A=DynamicParameter() {defaultValue=<rgba 0 1 1 1>, paramNames [PanSpeed, MipBias, SplineThicknessMultiplier, Param4]}.Param4
   B=ScalarParameter() {defaultValue=1, group=Other Settings, parameterName=Distortion Multiplier}
  )
  Reroute(None=Add(A=@n34.WorldPositionOffset, B=Multiply(A=@n46, B=@n47)))
 ]
}
